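# WebTexture_M is a UI-domain material in the engine's WebBrowser content, apparently for showing web browser texture output in UI.
# Its only node is a 2D texture parameter named SlateUI, defaulting to WebTexture_T.
# The parameter uses the External sampler type, which reads external textures rather than standard color textures.
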Material /Engine/WebBrowser/WebTexture_M
graph {
 properties {materialDomain=MD_UI, pixelDepthOffsetMode=PDOM_Legacy, usedShadingModels=Default Lit}
 nodes [TextureSampleParameter2D() {parameterName=SlateUI, samplerType=SAMPLERTYPE_External, texture=&/WebTexture_T}]
}
